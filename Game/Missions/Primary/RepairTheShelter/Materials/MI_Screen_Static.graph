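# Material for animated static on monitor screens in the RepairTheShelter mission, built as a base Material despite its MI_ name.
# Panned samples of one static texture, one distorting another's UVs, are tinted, summed and masked by that texture's alpha.
# A static switch adds optional sine flicker to the static.
# Another switch overlays a monitor graphics texture, pulsed by remapped sines and a mouth-mask texture sampled at time-driven UVs.
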Material /Game/Missions/Primary/RepairTheShelter/Materials/MI_Screen_Static
graph {
 properties {
  bUsedWithMeshDeformer=true
  bUsedWithNanite=true
  bUsedWithSkeletalMesh=true
  pixelDepthOffsetMode=PDOM_Legacy
  usedShadingModels=Default Lit
 }
 nodes [
  Constant() {r=0.5}
  Constant() {r=0.1}
  n1 = Multiply(A=Time(), B=Constant() {r=0.4})
  n2 = Constant() {r=0.6}
  n3 = Constant() {r=1}
  n4 = Time()
  n5 = Constant() {r=0.35}
  n6 = TextureSample(UVs=MaterialFunctionCall!(Sine Phase=Multiply(A=@n4, B=@n5), Value 1=Constant(), Value 2=Constant() {r=1}) {materialFunction=/Engine/Functions/Engine_MaterialFunctions02/Utility/Sine_Remapped}.Remapped Sine) {samplerType=SAMPLERTYPE_Masks, texture=/Game/Missions/Primary/RepairTheShelter/Textures/T_FemRay_MouthMask_Anim}
  n7 = Clamp(None=Multiply(
    A=MaterialFunctionCall(Sine Phase=@n1, Value 1=@n2, Value 2=@n3) {materialFunction=&Sine_Remapped}.Remapped Sine
    B=Add(A=@n6.G) {constB=0.85}
   ))
  n8 = TextureSampleParameter2D() {
   parameterName=Monitor Graphics Texture
   samplerType=SAMPLERTYPE_LinearColor
   texture=/Game/Missions/Primary/RepairTheShelter/Textures/Cinematic_Left_of_Computer_Monitor_e
  }
  n9 = TextureSample() {samplerType=SAMPLERTYPE_Masks, texture=/Game/Missions/Primary/RepairTheShelter/Textures/Cinematic_Screen_Static}
  n10 = Multiply(
   A=VectorParameter() {desc=screen 4, parameterName=Screen 4 Color, defaultValue=<rgba 0.1718 0.3126 0.415 1>}.RGB
   B=ScalarParameter() {defaultValue=1}
  )
  n11 = TextureCoordinate() {coordinateIndex=2}
  n12 = Panner(Coordinate=@n11, Time=Multiply(A=Time(), B=ScalarParameter() {defaultValue=-0.1, parameterName=Panner Speed 1, sliderMax=1, sliderMin=-0.1})) {desc=Panner 1, speedY=3}
  n13 = Multiply(A=Time(), B=ScalarParameter() {defaultValue=-0.5, parameterName=Panner Speed 2, sliderMax=-1, sliderMin=-0.5})
  n14 = TextureSample(UVs=Add(
    A=TextureSample(UVs=@n12) {samplerType=SAMPLERTYPE_Masks, texture=&Cinematic_Screen_Static}.G
    B=Panner(Coordinate=@n11, Time=@n13) {desc=Panner 2, speedY=0.2}
   )) {samplerType=SAMPLERTYPE_Masks, texture=&Cinematic_Screen_Static}
  n15 = VectorParameter() {desc=screen 2, parameterName=Screen 2 Color, defaultValue=<rgba 0.2925 0.325 0.305 1>}
  n16 = Time()
  n17 = ScalarParameter() {defaultValue=-0.6, parameterName=Panner Speed 3, sliderMax=-1, sliderMin=-0.6}
  n18 = TextureSample(UVs=Panner(Coordinate=@n11, Time=Multiply(A=@n16, B=@n17)) {desc=Panner 3, speedY=0.4}) {samplerType=SAMPLERTYPE_Masks, texture=&Cinematic_Screen_Static}
  n19 = VectorParameter() {desc=screen 3, parameterName=Screen 3 Color, defaultValue=<rgba 0.3715 0.415 0.344 1>}
  n20 = Multiply(A=Constant() {r=2}, B=Add(A=Multiply(A=@n14.G, B=@n15.RGB), B=Multiply(A=@n18.B, B=@n19.RGB)))
  n21 = Multiply(
   A=Time() {bIgnorePause=true, bOverride_Period=true, desc=Flicker Rate, period=1}
   B=ScalarParameter() {defaultValue=1, parameterName=Flicker Speed, sliderMax=12, sliderMin=0.01}
  )
  n22 = Multiply(
   A=Multiply(A=Multiply(A=@n9.A, B=@n10), B=Clamp(None=@n20))
   B=StaticSwitchParameter(True=Sine(None=@n21), False=ScalarParameter() {defaultValue=1, parameterName=Flicker on/off}) {parameterName=Flickr on/off}
  )
  StaticSwitchParameter(True=Add(A=Multiply(A=@n7, B=@n8.RGB), B=@n22), False=@n22) {parameterName=Monitor Graphics}
 ]
}
